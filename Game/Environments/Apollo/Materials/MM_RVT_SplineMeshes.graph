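# Spline mesh material for the Apollo environment that blends into the landscape by writing to and sampling its runtime virtual texture.
# Diffuse, Spec and world-space normal textures form the base attributes, which feed the RVT output with a tiled, offset Mask texture as opacity, zeroed on mobile.
# A static switch and FN_RVT_Switch_Material select between sampled RVT and those base attributes, and the result feeds the Apollo snow progression function.
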Material /Game/Environments/Apollo/Materials/MM_RVT_SplineMeshes
graph {
 properties {
  bTangentSpaceNormal=false
  bUseMaterialAttributes=true
  bUsedWithSplineMeshes=true
  pixelDepthOffsetMode=PDOM_Legacy
  usedShadingModels=Default Lit
 }
 nodes [
  n1 = RuntimeVirtualTextureSampleParameter() {
   group=RVT base
   materialType=BaseColor_Normal_Specular_YCoCg
   parameterName=RVT-Param
   virtualTexture=/Game/Athena/Apollo/Maps/Landscape/VT/RVT_Apollo_Landscape
  }
  n2 = SetMaterialAttributes!(MaterialAttributes, Base Color=@n1.BaseColor, Specular=@n1.BaseColor, Roughness=@n1.BaseColor, Normal=@n1.BaseColor) {attributeSetTypes [
    #1
    #2
    #3
    #4
   ]}
  n3 = TextureSampleParameter2D() {group=Texture, parameterName=Spec, samplerType=SAMPLERTYPE_Masks, texture=&T_Fortnite_Default_S}
  n4 = SetMaterialAttributes(
   Base Color=TextureSampleParameter2D() {group=Texture, parameterName=Diffuse, texture=/Game/Global/Textures/Default/Blanks/T_MidgreysRGB}.RGB
   Specular=@n3.R
   Roughness=@n3.R
   Normal=Transform(None=TextureSampleParameter2D() {group=Texture, parameterName=Normal, samplerType=SAMPLERTYPE_Normal, texture=/Engine/EngineMaterials/FlatNormal}.RGB)
  ) {attributeSetTypes [
    #1
    #2
    #3
    #4
   ]}
  n5 = MaterialFunctionCall(NoRVT=@n4, RVT=StaticSwitchParameter(True=@n2, False=@n4) {defaultValue=true, group=RVT base, parameterName=RVT Enable}) {materialFunction=&FN_RVT_Switch_Material}
  MaterialFunctionCall!(
   IsStaticSnow
   UseAlternateStaticSnowMask
   AlternateStaticSnowMask
   "UseRoad&&OcclusionMask"
   IsForGrassMeshToPushDown
   AdditionalMaskToRemoveSnow
   IsFallingSnowFX
   UseMinimapCoords
   IncreaseSnowUsingSlope
   UseSmoothBorder
   LerpHighM
   TemporalHeightLerpContrast
   IsLobbyIsland
   UseBR-IslandMask
   IsTree
   IsForRenderTarget
   UseSparkle
   UseSnowNormals
   UseAdditionalMaskToRemoveSnow
   IsTangentSpaceInput
   UseVertexNormalsForBasicSnow
   UseTangentSnowNormals
   Material=ShadingPathSwitch(Default=@n5.Result, Deferred=@n5.Result, Forward=@n5.Result, Mobile=@n5.Result)
  ) {materialFunction=/Game/Athena/Apollo/Environments/Landscape/Materials/Biome/UniqueBiomeMaterialFunctions/MF_Apollo_Snow_Progression}
  MaterialProxyReplace(Realtime=@n5.Result, MaterialProxy=@n4)
  n6 = GetMaterialAttributes(None=@n4) {attributeGetTypes [
    #1
    #2
    #3
    #4
   ]}
  n7 = ScalarParameter() {defaultValue=1, group=Mask, parameterName=Scale U, sliderMax=16, sliderMin=-16}
  n8 = ScalarParameter() {defaultValue=1, group=Mask, parameterName=Scale V, sliderMax=16, sliderMin=-16}
  n9 = Add(
   A=Multiply(A=TextureCoordinate(), B=AppendVector(A=@n7, B=@n8))
   B=AppendVector(
    A=ScalarParameter() {defaultValue=1, group=Mask, parameterName=Offset U, sliderMax=16, sliderMin=-16}
    B=ScalarParameter() {defaultValue=1, group=Mask, parameterName=Offset V, sliderMax=16, sliderMin=-16}
   )
  )
  RuntimeVirtualTextureOutput!(
   BaseColor=@n6.Base Color
   Specular=@n6.Base Color
   Roughness=@n6.Base Color
   Normal=@n6.Base Color
   WorldHeight
   Opacity=ShadingPathSwitch(
    Default=TextureSampleParameter2D(UVs=@n9) {parameterName=Mask, texture=&T_MidgreysRGB}.R
    Mobile=Constant()
   )
   Mask
   Displacement
   Mask4
  )
  n10 = VertexColor()
  StaticComponentMaskParameter(None=@n10) {defaultR=true, parameterName=MaskChannel}
  StaticSwitchParameter(False=@n10) {parameterName=UseTextureAsMask}
 ]
}
