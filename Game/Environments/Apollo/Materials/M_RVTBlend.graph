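# Apollo environment surface material for static, instanced and spline meshes that blends its textures into the landscape's runtime virtual texture.
# Diffuse, spec and normal parameters form the base attributes, blended with the RVT sample by a selectable vertex color channel.
# RVT Enable and FN_RVT_Switch_Material gate the blend; forward, mobile and material proxy paths use only the base.
# The result passes through MF_Apollo_Snow_Progression, with normals output in non-tangent space.
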Material /Game/Environments/Apollo/Materials/M_RVTBlend
graph {
 properties {
  bTangentSpaceNormal=false
  bUseMaterialAttributes=true
  bUsedWithInstancedStaticMeshes=true
  bUsedWithSplineMeshes=true
  pixelDepthOffsetMode=PDOM_Legacy
  usedShadingModels=Default Lit
 }
 nodes [
  n1 = TextureSampleParameter2D() {parameterName=Spec, samplerType=SAMPLERTYPE_Masks, texture=&T_Fortnite_Default_S}
  n2 = SetMaterialAttributes!(
   MaterialAttributes
   Base Color=TextureSampleParameter2D() {parameterName=Diffuse, texture=/Game/Global/Textures/Default/Blanks/T_MidgreysRGB}.RGB
   Specular=@n1.R
   Roughness=@n1.R
   Normal=Transform(None=TextureSampleParameter2D() {parameterName=Normal, samplerType=SAMPLERTYPE_Normal, texture=/Engine/EngineMaterials/FlatNormal}.RGB)
  ) {attributeSetTypes [
    #1
    #2
    #3
    #4
   ]}
  n3 = RuntimeVirtualTextureSample!(UVs, World Position) {materialType=BaseColor_Normal_Specular_YCoCg, virtualTexture=/Game/Athena/Apollo/RVT/RVT_Apollo_Landscape}
  n4 = SetMaterialAttributes(Base Color=@n3.BaseColor, Specular=@n3.BaseColor, Roughness=@n3.BaseColor, Normal=@n3.BaseColor) {attributeSetTypes [
    #1
    #2
    #3
    #4
   ]}
  n5 = VertexColor()
  n6 = StaticComponentMaskParameter(None=@n5) {defaultR=true, parameterName=MaskChannel}
  n7 = StaticSwitchParameter(
   True=MaterialFunctionCall(NoRVT=@n2, RVT=BlendMaterialAttributes(A=@n2, B=@n4, Alpha=@n6)) {materialFunction=&FN_RVT_Switch_Material}.Result
   False=@n2
  ) {defaultValue=true, group=RVT base, parameterName=RVT Enable}
  MaterialFunctionCall!(
   IsStaticSnow
   UseAlternateStaticSnowMask
   AlternateStaticSnowMask
   "UseRoad&&OcclusionMask"
   IsForGrassMeshToPushDown
   AdditionalMaskToRemoveSnow
   IsFallingSnowFX
   UseMinimapCoords
   IncreaseSnowUsingSlope
   UseSmoothBorder
   LerpHighM
   TemporalHeightLerpContrast
   IsLobbyIsland
   UseBR-IslandMask
   IsTree
   IsForRenderTarget
   UseSparkle
   UseSnowNormals
   UseAdditionalMaskToRemoveSnow
   IsTangentSpaceInput
   UseVertexNormalsForBasicSnow
   UseTangentSnowNormals
   Material=ShadingPathSwitch(Default=MaterialProxyReplace(Realtime=@n7, MaterialProxy=@n2), Forward=@n2, Mobile=@n2)
  ) {materialFunction=/Game/Athena/Apollo/Environments/Landscape/Materials/Biome/UniqueBiomeMaterialFunctions/MF_Apollo_Snow_Progression}
  StaticSwitchParameter(False=@n5) {parameterName=UseTextureAsMask}
 ]
}
